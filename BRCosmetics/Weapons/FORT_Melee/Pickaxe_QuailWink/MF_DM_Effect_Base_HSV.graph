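# Material Designer effect function, stored with the QuailWink pickaxe cosmetic, that adjusts the hue, saturation and value of a Base color input.
# It converts Base to HSV, offsets hue in degrees with wraparound, and remaps saturation and value from 0–2 controls centered on one.
# Branches scale toward zero or push toward full by the sign of each offset.
# The recombined RGB blends with the original by Effect Amount as a percentage.
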MaterialFunction /BRCosmetics/Weapons/FORT_Melee/Pickaxe_QuailWink/MF_DM_Effect_Base_HSV
graph {
 properties {bExposeToLibrary=true, userExposedCaption=Hue Saturation Value, libraryCategoriesText [Misc, Motion Design, Material Designer]}
 nodes [
  n1 = MaterialFunctionCall(X=NamedRerouteUsage(), Y=NamedRerouteUsage(), Z=NamedRerouteUsage()) {materialFunction=&MakeFloat3}
  n2 = FunctionInput() {
   bUsePreviewValueAsDefault=true
   desc="UIMin=0, UIMax=100, ClampMin=0, ClampMax=100"
   inputName=Effect Amount
   inputType=FunctionInput_Scalar
   sortPriority=99
   previewValue=<100 0 0 1>
  }
  n3 = Divide(A=@n2) {constB=100}
  FunctionOutput(None=LinearInterpolate(A=NamedRerouteUsage(), B=HsvToRgb!(None=@n1.Result), Alpha=Saturate(None=@n3)))
  n4 = FunctionInput() {
   desc="UIMin=0, UIMax=360, ClampMin=0, ClampMax=360"
   description=Values from -360-360 which change the color of the image.
   inputName=Hue
   inputType=FunctionInput_Scalar
   sortPriority=1
  }
  NamedRerouteDeclaration(None=@n4) {name=HueInput, nodeColor=<rgba 0 1 0.8044 1>}
  n5 = MaterialFunctionCall(Float3=MaterialFunctionCall!(In=NamedRerouteUsage()) {materialFunction=/Engine/Functions/Engine_MaterialFunctions01/Chromakeying/RGBtoHSV}.Result) {materialFunction=&BreakOutFloat3Components}
  NamedRerouteDeclaration(None=@n5.R) {name=InitialHue, nodeColor=<rgba 0.7109 0 1 1>}
  NamedRerouteDeclaration(None=@n5.G) {name=InitialSaturation, nodeColor=<rgba 0.2891 0 1 1>}
  NamedRerouteDeclaration(None=@n5.B) {name=InitialValue, nodeColor=<rgba 0.2891 0 1 1>}
  n6 = NamedRerouteUsage()
  n7 = Fmod(A=Add(A=NamedRerouteUsage(), B=Divide(A=@n6) {constB=360}), B=Constant() {r=1})
  NamedRerouteDeclaration(None=If(A=@n7, "A > B"=@n7, "A == B"=@n7, "A < B"=Add(A=@n7))) {name=ModifiedHue, nodeColor=<rgba 1 0.0644 0 1>}
  n8 = NamedRerouteUsage()
  n9 = Add(A=NamedRerouteUsage())
  n10 = NamedRerouteUsage()
  n11 = Multiply(A=OneMinus(None=@n10), B=NamedRerouteUsage())
  NamedRerouteDeclaration(None=If(A=NamedRerouteUsage(), "A > B"=Add(A=@n10, B=@n11), "A == B"=NamedRerouteUsage(), "A < B"=Multiply(A=@n8, B=@n9))) {name=ModifiedSaturation, nodeColor=<rgba 1 0.0644 0 1>}
  n12 = NamedRerouteUsage()
  n13 = Add(A=NamedRerouteUsage())
  n14 = NamedRerouteUsage()
  n15 = Multiply(A=OneMinus(None=@n14), B=NamedRerouteUsage())
  NamedRerouteDeclaration(None=If(A=NamedRerouteUsage(), "A > B"=Add(A=@n14, B=@n15), "A == B"=NamedRerouteUsage(), "A < B"=Multiply(A=@n12, B=@n13))) {name=ModifiedValue, nodeColor=<rgba 1 0.0644 0 1>}
  NamedRerouteDeclaration(None=FunctionInput(Preview=TextureSample() {texture=/Engine/EngineMaterials/DefaultDiffuse}.RGB) {inputName=Base, previewValue=<1 0 0 1>}) {name=Stage, nodeColor=<rgba 0 1 0.8044 1>}
  n16 = FunctionInput() {
   bUsePreviewValueAsDefault=true
   desc="UIMin=0, UIMax=2, ClampMin=0, ClampMax=2"
   description=Increase or decrease the saturation level. Values between 0.0 and 2.0.
   inputName=Saturation
   inputType=FunctionInput_Scalar
   sortPriority=2
   previewValue=<1 0 0 1>
  }
  NamedRerouteDeclaration(None=Subtract(A=Clamp(None=@n16) {maxDefault=2})) {name=SaturationInput, nodeColor=<rgba 0 1 0.8044 1>}
  n17 = FunctionInput() {
   bUsePreviewValueAsDefault=true
   desc="UIMin=0, UIMax=2, ClampMin=0, ClampMax=2"
   description=Increase or decrease the Value level. Values between 0.0 and 2.0.
   inputName=Value
   inputType=FunctionInput_Scalar
   sortPriority=3
   previewValue=<1 0 0 1>
  }
  NamedRerouteDeclaration(None=Subtract(A=Clamp(None=@n17) {maxDefault=2})) {name=ValueInput, nodeColor=<rgba 0 1 0.8044 1>}
 ]
}
