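# Translucent, unlit, two-sided glow material for the Quail Wink pickaxe FX, usable on Niagara sprites and mesh particles.
# Color comes from a gradient texture indexed by Perlin noise and masked time scrolling, with optional tint and particle-driven hue shift.
# Opacity blends a glow texture with view-angle fade, deploy wipe and quality-dependent depth fade against a procedural UV gradient shape via GlowShape.
# A SplineThicken function adds width controls for spline meshes.
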Material /BRCosmetics/Weapons/FORT_Melee/Pickaxe_QuailWink/FX/M_QuailWink_Glow
graph {
 properties {
  bUsedWithNiagaraMeshParticles=true
  bUsedWithNiagaraSprites=true
  blendMode=BLEND_Translucent
  pixelDepthOffsetMode=PDOM_Legacy
  shadingModel=MSM_Unlit
  twoSided=true
  usedShadingModels=Unlit
 }
 nodes [
  n1 = ParticleColor()
  n2 = TextureSampleParameter2D() {
   parameterName=Glow Texture
   samplerType=SAMPLERTYPE_LinearGrayscale
   texture=&/T_QuailWink_Glow
  }
  n3 = DotProduct(A=CameraVectorWS(), B=VertexNormalWS())
  n4 = Multiply(
   A=Multiply(A=@n1.A, B=@n2.R)
   B=SmoothStep(
    Min=ScalarParameter() {parameterName=View Angle Fade Min, sliderMax=1}
    Max=ScalarParameter() {defaultValue=1, parameterName=View Angle Fade Max, sliderMax=1}
    Value=Abs(None=@n3)
   )
  )
  n5 = DepthFade(Opacity=@n4, FadeDistance=ScalarParameter() {defaultValue=1, parameterName=FadeDistance})
  n6 = ScalarParameter() {defaultValue=0.2, parameterName=GradientMask_Width}
  n7 = MaterialFunctionCall(Float2=TextureCoordinate()) {materialFunction=&BreakOutFloat2Components}
  n8 = ScalarParameter() {defaultValue=0.5, parameterName=GradientMask_Size_U}
  n9 = SmoothStep(
   Min=@n6
   Max=Add(A=@n6, B=ScalarParameter() {defaultValue=0.75, parameterName=GradientMask_Hardness})
   Value=Multiply(A=SphereMask(A=@n7.R, B=@n8, Radius=@n8) {hardnessPercent=0})
  )
  n10 = ScalarParameter() {defaultValue=0.3, parameterName=GradientMaskSide_Offset}
  n11 = SmoothStep(Min=@n10, Max=Add(A=@n10, B=ScalarParameter() {defaultValue=1, parameterName=GradientMaskSide_Hardness}), Value=@n7.R)
  n12 = ScalarParameter() {parameterName=UseMaskForSideGlow}
  n13 = ScalarParameter() {parameterName=GradientMask_Width_U}
  n14 = Add(A=@n13, B=ScalarParameter() {defaultValue=0.5, parameterName=GradientMask_Hardness_U})
  n15 = ScalarParameter() {defaultValue=0.5, parameterName=GradientMask_Size_V}
  n16 = SphereMask(A=@n7.G, B=@n15, Radius=@n15) {hardnessPercent=0}
  n17 = Power(Base=Multiply(A=LinearInterpolate(A=@n9, B=@n11, Alpha=@n12) {constAlpha=0}, B=SmoothStep(Min=@n13, Max=@n14, Value=@n16)))
  n18 = Multiply(
   A=Multiply(A=Multiply(A=@n17), B=ScalarParameter() {defaultValue=0.15, parameterName=Opacity Mult})
   B=MaterialFunctionCall("FadeSensitivity (optional)"=ScalarParameter() {defaultValue=8, parameterName=Angle Fade}) {materialFunction=&FadeBasedOnViewAngle}.Result
  )
  LinearInterpolate(
   A=Multiply(
    A=QualitySwitch(Default=@n5, Low=@n4, High=@n5, Medium=@n4, Epic=@n5)
    B=MaterialFunctionCall!(InDerectionalGradient) {materialFunction=/BRCosmetics/Weapons/FORT_Melee/Pickaxe_ClaimReflect/Materials/MF_ClaimReflect_DeployWipe}.WipeMask
   )
   B=LinearInterpolate(A=@n18, B=Multiply(A=@n1.A, B=@n18), Alpha=ScalarParameter() {parameterName=UseParticleColorAlpha})
   Alpha=ScalarParameter() {parameterName=GlowShape}
  )
  n19 = Multiply(B=TextureCoordinate()) {constA=0.06}
  n20 = Multiply(A=Time()) {constB=0.03}
  n21 = Power(Base=TextureSample(
    UVs=Add(A=@n19, B=@n20)
    Tex=TextureObject() {samplerType=SAMPLERTYPE_LinearColor, texture=/Game/Global/Textures/Default/Tileable_Patterns/T_PerlinNoise_Uncompressed}
   ).RGB) {constExponent=0.5}
  n22 = ScalarParameter() {defaultValue=10, group=Quail Wink / Loan Float / Evoke Find, parameterName=Noise Modulation - Strength}
  n23 = TextureSample(UVs=Multiply(
    A=TextureCoordinate()
    B=ScalarParameter() {defaultValue=1, group=Quail Wink / Loan Float / Evoke Find, parameterName=Noise Modulation - Tiling}
   )) {
   samplerSource=SSM_Wrap_WorldGroupSettings
   samplerType=SAMPLERTYPE_LinearGrayscale
   texture=/BRCosmetics/Weapons/FORT_Melee/Pickaxe_QuailWink/T_ColorShift_02_Mask
  }
  n24 = Multiply(
   A=Add(A=Time(), B=Multiply(A=@n22, B=@n23.R))
   B=ScalarParameter() {defaultValue=0.02, group=Quail Wink / Loan Float / Evoke Find, parameterName=Noise Modulation - Speed}
  )
  n25 = TextureSampleParameter2D(UVs=ComponentMask(None=Subtract(A=@n21, B=@n24)) {r=true}) {
   group=Quail Wink / Loan Float / Evoke Find
   parameterName=Gradient Color
   texture=/BRCosmetics/Weapons/FORT_Melee/Pickaxe_QuailWink/T_ColorShift_02_Gradient
  }
  n26 = Multiply(A=VectorParameter() {parameterName=Tint, defaultValue=<rgba 1 0.4186 0.8952 1>}.RGB, B=@n25.RGB)
  n27 = ScalarParameter() {group=Quail Wink / Loan Float / Evoke Find, parameterName=UseTint}
  n28 = LinearInterpolate(Alpha=DynamicParameter() {defaultValue=<rgba 0 1 1 1>, paramNames [HueShiftAnim,Param2,Param3,Param4]}.HueShiftAnim) {constB=155}
  n29 = ScalarParameter() {group=Quail Wink / Loan Float / Evoke Find, parameterName=UseHueShift}
  MaterialFunctionCall(Emissive=MaterialFunctionCall!(
    Base=LinearInterpolate(A=@n25.RGB, B=@n26, Alpha=@n27)
    Hue=LinearInterpolate(B=@n28, Alpha=@n29)
    Saturation=ScalarParameter() {defaultValue=1.3, group=Quail Wink / Loan Float / Evoke Find, parameterName=Saturation}
    Value
    Effect Amount
   ) {materialFunction=/BRCosmetics/Weapons/FORT_Melee/Pickaxe_QuailWink/MF_DM_Effect_Base_HSV}.Result) {materialFunction=&MF_FX_EmissiveAdjustments}
  MaterialFunctionCall(
   WidthBase=ScalarParameter() {group=Spline Settings, parameterName=Spline Width Base, sortPriority=1}
   WidthTip=ScalarParameter() {group=Spline Settings, parameterName=Spline Width Tip, sortPriority=1}
  ) {materialFunction=&SplineThicken}
 ]
}
