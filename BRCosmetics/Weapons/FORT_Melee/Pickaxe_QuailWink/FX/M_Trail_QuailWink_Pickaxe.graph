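# Additive unlit Niagara ribbon material for the swing trail of the QuailWink pickaxe, drawing animated chromatic wave lines and starry dots.
# Three copies of a sine wave, each distorted by offset panning blurred noise, fill the R, G and B channels to split colors.
# A panning dots texture adds stars, scaled by the DotsAlpha dynamic parameter.
# Sphere-mask, noise and UV fades shape the trail, tinted by particle color through MF_FX_EmissiveAdjustments.
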Material /BRCosmetics/Weapons/FORT_Melee/Pickaxe_QuailWink/FX/M_Trail_QuailWink_Pickaxe
graph {
 properties {
  bAutomaticallySetUsageInEditor=false
  bUsedWithNiagaraRibbons=true
  blendMode=BLEND_Additive
  shadingModel=MSM_Unlit
  usedShadingModels=Unlit
 }
 nodes [
  n1 = Time()
  n2 = ScalarParameter() {defaultValue=-0.25, group=WaveSettings, parameterName=WaveSpeedMin}
  n3 = NamedRerouteUsage()
  n4 = LinearInterpolate(
   A=ScalarParameter() {defaultValue=17.5, group=WaveSettings, parameterName=WavePatternSizeTip}
   B=ScalarParameter() {defaultValue=10, group=WaveSettings, parameterName=WavePatternSize}
   Alpha=ComponentMask(None=NamedRerouteUsage()) {r=true}
  )
  n5 = ComponentMask(None=Add(A=Multiply(A=@n1, B=@n2), B=Multiply(A=@n3, B=@n4))) {r=true}
  n6 = Add(
   A=AppendVector(A=ScalarParameter() {parameterName=R_Offset_U}, B=ScalarParameter() {defaultValue=1.05, parameterName=R_Offset_V})
   B=TextureCoordinate()
  )
  n7 = AppendVector(A=ScalarParameter() {defaultValue=0.5, parameterName=DistortionTileX}, B=ScalarParameter() {defaultValue=0.5, parameterName=DistortionTileY})
  n8 = Constant() {r=-0.25}
  n9 = ScalarParameter() {defaultValue=-0.1, parameterName=DistortionSpeed}
  n10 = TextureSample(UVs=Panner(Coordinate=Multiply(A=@n6, B=@n7), Speed=AppendVector(A=@n8, B=@n9))) {samplerType=SAMPLERTYPE_Grayscale, texture=&LowResBlurredNoise}
  n11 = Time()
  n12 = Constant() {r=-2.5}
  n13 = Sine(None=Add(A=Add(A=@n5, B=@n10.G), B=Multiply(A=@n11, B=@n12)))
  n14 = Time()
  n15 = ScalarParameter() {defaultValue=-0.25, group=WaveSettings, parameterName=WaveSpeedMin}
  n16 = NamedRerouteUsage()
  n17 = LinearInterpolate(
   A=ScalarParameter() {defaultValue=17.5, group=WaveSettings, parameterName=WavePatternSizeTip}
   B=ScalarParameter() {defaultValue=10, group=WaveSettings, parameterName=WavePatternSize}
   Alpha=ComponentMask(None=NamedRerouteUsage()) {r=true}
  )
  n18 = ComponentMask(None=Add(A=Multiply(A=@n14, B=@n15), B=Multiply(A=@n16, B=@n17))) {r=true}
  n19 = Add(
   A=AppendVector(A=ScalarParameter() {parameterName=G_Offset_U}, B=ScalarParameter() {defaultValue=1, parameterName=G_Offset_V})
   B=TextureCoordinate()
  )
  n20 = AppendVector(A=ScalarParameter() {defaultValue=0.5, parameterName=DistortionTileX}, B=ScalarParameter() {defaultValue=0.5, parameterName=DistortionTileY})
  n21 = Constant() {r=-0.25}
  n22 = ScalarParameter() {defaultValue=-0.1, parameterName=DistortionSpeed}
  n23 = TextureSample(UVs=Panner(Coordinate=Multiply(A=@n19, B=@n20), Speed=AppendVector(A=@n21, B=@n22))) {samplerType=SAMPLERTYPE_Grayscale, texture=&LowResBlurredNoise}
  n24 = Time()
  n25 = Constant() {r=-2.5}
  n26 = Sine(None=Add(A=Add(A=@n18, B=@n23.G), B=Multiply(A=@n24, B=@n25)))
  n27 = Time()
  n28 = ScalarParameter() {defaultValue=-0.25, group=WaveSettings, parameterName=WaveSpeedMin}
  n29 = NamedRerouteUsage()
  n30 = LinearInterpolate(
   A=ScalarParameter() {defaultValue=17.5, group=WaveSettings, parameterName=WavePatternSizeTip}
   B=ScalarParameter() {defaultValue=10, group=WaveSettings, parameterName=WavePatternSize}
   Alpha=ComponentMask(None=NamedRerouteUsage()) {r=true}
  )
  n31 = ComponentMask(None=Add(A=Multiply(A=@n27, B=@n28), B=Multiply(A=@n29, B=@n30))) {r=true}
  n32 = Add(
   A=AppendVector(A=ScalarParameter() {parameterName=B_Offset_U}, B=ScalarParameter() {defaultValue=0.95, parameterName=B_Offset_V})
   B=TextureCoordinate()
  )
  n33 = AppendVector(A=ScalarParameter() {defaultValue=0.5, parameterName=DistortionTileX}, B=ScalarParameter() {defaultValue=0.5, parameterName=DistortionTileY})
  n34 = Constant() {r=-0.25}
  n35 = ScalarParameter() {defaultValue=-0.1, parameterName=DistortionSpeed}
  n36 = TextureSample(UVs=Panner(Coordinate=Multiply(A=@n32, B=@n33), Speed=AppendVector(A=@n34, B=@n35))) {samplerType=SAMPLERTYPE_Grayscale, texture=&LowResBlurredNoise}
  n37 = Time()
  n38 = Constant() {r=-2.5}
  n39 = Sine(None=Add(A=Add(A=@n31, B=@n36.G), B=Multiply(A=@n37, B=@n38)))
  n40 = MaterialFunctionCall(R=@n13, G=@n26, B=@n39) {materialFunction=&AppendMany}
  n41 = Multiply(A=Multiply(A=@n40.RGB), B=NamedRerouteUsage())
  n42 = ComponentMask(None=TextureCoordinate()) {r=true}
  n43 = Saturate(None=Multiply(A=OneMinus(None=@n42)) {constB=2})
  n44 = Multiply(
   A=SphereMask(
    A=TextureCoordinate() {uTiling=0}
    B=Constant2Vector() {g=0.5}
    Radius=ScalarParameter() {defaultValue=0.5, parameterName=Edge fade rang}
    Hardness=ScalarParameter() {defaultValue=0.5, parameterName=Edge fade hardness}
   ) {attenuationRadius=0.5, hardnessPercent=50}
   B=Power(Base=Multiply(A=@n36.G, B=@n43)) {constExponent=0.5}
  )
  n45 = Multiply(A=NamedRerouteUsage()) {constB=2}
  n46 = OneMinus(None=ComponentMask(None=TextureCoordinate()) {r=true})
  n47 = Multiply(A=DynamicParameter() {paramNames [DotsAlpha,Param2,Param3,Param4]}.DotsAlpha, B=Multiply(A=@n44, B=Multiply(A=@n45, B=@n46)))
  Add(A=Multiply(A=Saturate(None=@n41), B=ParticleColor().A), B=@n47)
  n48 = NamedRerouteUsage()
  MaterialFunctionCall(Emissive=Multiply(A=ParticleColor().RGB, B=Add(A=@n47, B=@n48))) {materialFunction=&MF_FX_EmissiveAdjustments}
  NamedRerouteDeclaration(None=Saturate(None=TextureCoordinate())) {name=WavePatternUVs, nodeColor=<rgba 1 0.6328 0 1>}
  n49 = ComponentMask(None=TextureCoordinate()) {r=true}
  n50 = Multiply(A=@n44, B=Saturate(None=Multiply(A=@n49) {constB=10}))
  n51 = NamedRerouteUsage()
  n52 = ComponentMask(None=TextureCoordinate()) {r=true}
  n53 = Multiply(A=@n50, B=Add(A=@n50, B=Multiply(A=@n51, B=@n52)))
  NamedRerouteDeclaration(None=Saturate(None=SmoothStep(Max=@n50, Value=@n53))) {name=TrailMask, nodeColor=<rgba 0 1 0.25 1>}
  NamedRerouteDeclaration(None=Saturate(None=TextureCoordinate())) {name=WavePatternUVs_1, nodeColor=<rgba 1 0.6328 0 1>}
  NamedRerouteDeclaration(None=Saturate(None=TextureCoordinate())) {name=WavePatternUVs_2, nodeColor=<rgba 1 0.6328 0 1>}
  n54 = TextureCoordinate()
  n55 = AppendVector(A=ScalarParameter() {defaultValue=0.75, parameterName=StarsTileX}, B=ScalarParameter() {defaultValue=0.25, parameterName=StarsTileY})
  n56 = Constant() {r=-0.1}
  n57 = ScalarParameter() {defaultValue=0.05, parameterName=StarsSpeed}
  n58 = TextureSample(UVs=Panner(Coordinate=Multiply(A=@n54, B=@n55), Speed=AppendVector(A=@n56, B=@n57))) {texture=/Game/Athena/Environments/Nexus/Textures/T_Tile_Dots}
  n59 = Add(A=@n58.R, B=@n58.R)
  n60 = Multiply(A=@n59, B=@n40.RGB)
  n61 = MaterialFunctionCall(
   R=Saturate(None=ConstantBiasScale(None=@n13) {bias=-0.9, scale=2})
   G=Saturate(None=ConstantBiasScale(None=@n26) {bias=-0.9, scale=2})
   B=Saturate(None=ConstantBiasScale(None=@n39) {bias=-0.9, scale=2})
  ) {materialFunction=&AppendMany}
  NamedRerouteDeclaration(None=Add(A=Saturate(None=@n60), B=Multiply(A=@n61.RGB))) {name=ChromaLines, nodeColor=<rgba 0 1 0.0859 1>}
  NamedRerouteDeclaration(None=@n59) {name=DotsStars, nodeColor=<rgba 1 0 0.5625 1>}
 ]
}
