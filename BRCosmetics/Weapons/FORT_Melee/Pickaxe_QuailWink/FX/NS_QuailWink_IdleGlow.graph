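# Idle glow effect for the QuailWink pickaxe, built from one Glow emitter with five sprite renderers.
# A burst spawns two persistent particles whose spawn group sets visibility tag, size, offset and facing.
# Extra attributes place capsule, hand and sprite glows along local X, pulsing blue and purple via a waveform.
# Alpha ramps in with system age, scaled differently for front end and gameplay; the system's X axis feeds material parameters.
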NiagaraSystem /BRCosmetics/Weapons/FORT_Melee/Pickaxe_QuailWink/FX/NS_QuailWink_IdleGlow
dependencies {
 usedDataInterfaces [&NiagaraDataInterfaceCurve]
 usedDynamicInputs [&FNTimeOfDayScalar]
 usedModules [
  &SystemState
  &self:SystemSpawnScript.SystemScriptSource.SystemScriptGraph.NiagaraNodeAssignment_1.SetVariables_#1
 ]
}
emitters {Glow {
  emitterUpdateScript [
   &EmitterState(
    Inactive Response=::NewEnumerator1
    Life Cycle Mode=::NewEnumerator1
    Loop Behavior=::NewEnumerator1
    Loop Duration Mode=::NewEnumerator1
    UseLoopDelay=false
   )
   &SpawnBurst_Instantaneous(Spawn Count=2, Use Spawn Probability=false, UseLoopCountLimit=false)
  ]
  particleSpawnScript [
   &FN_SetSpawnGroup(Number of Spawn Groups=2)
   /SetVariables(Particles.VisibilityTag=$Particles.SpawnGroup) {isSetParameters=true}
   &InitializeParticle(
    Color=<rgba 0.2135 0.6982 1 0.5>
    Lifetime=5
    Position Offset=&MergeHelper_VectorSwitch(0=<-10 2 0>, 1=<-5 -10 0>, 2=?, 3=?, Index=$Particles.VisibilityTag, Num=::NewEnumerator0)
    Position Offset Coordinate Space=::Local
    Sprite Size=&MergeHelper_Vector2Switch(0=<80 90>, 1=<20 90>, 2=?, 3=?, 4=?, Index=$Particles.VisibilityTag, Num=::NewEnumerator0)
    Sprite Size Mode=::NewEnumerator4
    UsePositionOffset=true
    Write Material Random=true
   )
   /SetVariables(Particles.CapsuleCameraOffset=1, Particles.CapsuleHandSize:Vector2f=<17.5 1>, Particles.CapsuleSize:Vector2f=<10 1>) {isSetParameters=true}
  ]
  particleUpdateScript [
   &ParticleState(
    Kill Particles When Lifetime Has Elapsed=false
    Let Infinitely Lived Particles Die When Emitter Deactivates=true
    Loop Particles Lifetime=true
   )
   /Niagara/Noise/Waveform(
    EnableExponent1=false
    GlobalClampMax=false
    GlobalClampMin=false
    UseInput1=false
    UsePhase1=false
    "Waveform [1]":ENiagara_Waveforms=::NewEnumerator12
    "[1] Frequency"=0.5
   )
   &SpriteFacingAndAlignment(
    Sprite Alignment=<1 0 0>
    Sprite Facing=&MergeHelper_VectorSwitch(0=<0 0 1>, 1=<-0.3 1 0>, 2=?, 3=?, Index=$Particles.VisibilityTag, Num=::NewEnumerator0)
    WriteAlignment=true
    WriteFacing=true
   )
   &ScaleColor(
    Scale Alpha=&FloatFromCurve(
     FloatCurve=DataInterface() {propertyValues {Curve {
        defaultValue=0
        keys [
         {arriveTangent=4.1785, interpMode=RCIM_Cubic, leaveTangent=4.1785, tangentMode=RCTM_User, time=0, value=0}
         {arriveTangent=1.1179, interpMode=RCIM_Cubic, leaveTangent=1.1179, tangentMode=RCTM_User, time=0.2042, value=0.6228}
         {arriveTangent=0.0224, interpMode=RCIM_Cubic, leaveTangent=0.0224, tangentMode=RCTM_User, time=1, value=1}
        ]
       }}}
     CurveIndex=$Engine.System.Age
     Scale Curve=&FN_FrontEndFloatModifier(Gameplay=0.25, Front End=0.15)
    )
    ScaleA=true
    ScaleRGB=true
   )
   &DynamicMaterialParameters(
    Index 0 Param 1=&MakeFloatFromVector(Vector=$Engine.Owner.SystemXAxis, Channel=::NewEnumerator0)
    Index 0 Param 2=&MakeFloatFromVector(Vector=$Engine.Owner.SystemXAxis, Channel=::NewEnumerator1)
    Index 0 Param 3=&MakeFloatFromVector(Vector=$Engine.Owner.SystemXAxis, Channel=::NewEnumerator2)
    Index 0 Param 4=30
    Module.Param0WriteEnabled=true
    Module.Param1WriteEnabled=true
    Module.Param2WriteEnabled=true
    Module.Param3WriteEnabled=true
   )
   /SetVariables(
    Particles.CapsuleColor:LinearColor=&Multiply_LinearColorByFloat(
     Linear Color=&LerpLinearColors(StartColor=<rgba 0.7253 0.25 1 0.5>, EndColor=<rgba 0.2135 0.6982 1 0.5>, LerpFactor=$Particles.Waveform.Waveform)
     Float=0.8
    )
    Particles.CapsuleHandPosition:NiagaraPosition=&AddVectorToPosition(Position=$Particles.Position, Vector=&TransformVector(Vector=<-25 0.5 0>, Destination Space=::Simulation))
    Particles.CapsulePosition:NiagaraPosition=&AddVectorToPosition(Position=$Particles.Position, Vector=&TransformVector(Vector=<-99 0.5 0>, Destination Space=::Simulation))
    Particles.DMP_HandGlow:Vector4f=&MakeVector4(
     X=&MakeFloatFromVector(Vector=$Engine.Owner.SystemXAxis, Channel=::NewEnumerator0)
     Y=&MakeFloatFromVector(Vector=$Engine.Owner.SystemXAxis, Channel=::NewEnumerator1)
     Z=&MakeFloatFromVector(Vector=$Engine.Owner.SystemXAxis, Channel=::NewEnumerator2)
     W=15
    )
   ) {isSetParameters=true}
   &FN_Sprite_Attributes(
    Camera Offset=-5
    Color=&MultiplyLinearColor(
     Linear Color=&LerpLinearColors(
      StartColor=<rgba 0.2135 0.6982 1 0.5>
      EndColor=<rgba 0.7253 0.25 1 0.5>
      LerpFactor=&NormalizeFloat(
       Value=&Add_Float(
        A=$Particles.Waveform.Waveform
        B=&FloatFromCurve(FloatCurve=DataInterface() {propertyValues {Curve {
            defaultValue=1
            keys [
             {interpMode=RCIM_Cubic, tangentMode=RCTM_User, time=0, value=-0.5}
             {interpMode=RCIM_Cubic, tangentMode=RCTM_User, time=0.2119, value=-0.1706}
             {arriveTangent=-0.3322, interpMode=RCIM_Cubic, leaveTangent=-0.3322, time=0.4361, value=-0.5132}
             {arriveTangent=0.0235, interpMode=RCIM_Cubic, leaveTangent=0.0235, time=0.659, value=-0.3191}
             {tangentMode=RCTM_User, time=1, value=-0.5}
            ]
           }}})
       )
       Min=0
       Max=1
      )
     )
     Scale Factor=<0.8 0.8 0.8 0.075>
    )
    Create Camera Offset Attribute=true
    Create Color Attribute=true
    Create Position Attribute=true
    Create SpriteSize Attribute=true
    Position:NiagaraPosition=&AddVectorToPosition(Position=$Particles.Position, Vector=&TransformVector(Vector=<-122.5 2 0>, Destination Space=::Simulation))
    SpriteSize=<17 17>
    Write to Module Namespace=true
   )
  ]
  rendererClasses [&NiagaraSpriteRendererProperties]
  renderers [
   {rendererClass=&NiagaraSpriteRendererProperties}
   {rendererClass=&NiagaraSpriteRendererProperties, rendererIndex=1}
   {rendererClass=&NiagaraSpriteRendererProperties, rendererIndex=2}
   {rendererClass=&NiagaraSpriteRendererProperties, rendererIndex=3}
   {rendererClass=&NiagaraSpriteRendererProperties, rendererIndex=4}
  ]
 }}
summary {
 systemName=NS_QuailWink_IdleGlow
 emitters [{emitterName=Glow, rendererClasses [&NiagaraSpriteRendererProperties]}]
 userVariables [
  User.bIsFrontEnd {type=NiagaraBool}
  User.bIsFrontEndPreview {type=NiagaraBool}
 ]
}
